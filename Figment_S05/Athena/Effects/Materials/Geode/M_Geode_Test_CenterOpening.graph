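# Translucent, two-sided unlit mesh-particle material for a Geode center-opening test effect: swirling colored energy with refraction.
# A UV2 sphere mask blends opposite-panning turbulence color and normal samples, shaped by UV0 V gradients.
# Turbulence drives opacity and distorts a rotating rainbow added to crack-function colors, with normal-dot highlights.
# Vertex red boosts opacity, green blends toward InnerColor; Low quality drops rainbow, highlights, normals and refraction.
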Material /Figment_S05/Athena/Effects/Materials/Geode/M_Geode_Test_CenterOpening
graph {
 properties {
  bUsedWithMeshParticles=true
  blendMode=BLEND_Translucent
  pixelDepthOffsetMode=PDOM_Legacy
  shadingModel=MSM_Unlit
  twoSided=true
  usedShadingModels=Unlit
 }
 nodes [
  n1 = ParticleColor()
  n2 = VertexColor()
  n3 = SphereMask(A=ComponentMask(None=TextureCoordinate() {coordinateIndex=2}) {g=true}, B=Constant()) {attenuationRadius=0.5, hardnessPercent=70}
  n4 = LinearInterpolate(Alpha=@n3) {constA=-1}
  n5 = Constant2Vector() {g=0.05}
  n6 = ComponentMask(None=TextureCoordinate()) {g=true}
  n7 = Power(Base=@n6) {constExponent=3}
  n8 = Add(A=TextureCoordinate() {coordinateIndex=2, vTiling=0.3}, B=Multiply(A=@n4, B=Multiply(A=@n5, B=@n7))) {constB=0.1}
  n9 = Multiply(A=Time(), B=ScalarParameter() {defaultValue=0.5, parameterName=PanSpeed})
  n10 = Panner(Coordinate=@n8, Time=@n9) {speedX=0.007, speedY=-0.04}
  n11 = Panner(Coordinate=@n8, Time=@n9) {speedX=-0.007, speedY=0.04}
  n12 = Multiply(A=LinearInterpolate(
    A=TextureSample(UVs=@n10) {texture=/Figment_S05/Effects/Fort_Effects/Textures/Plasma/T_Turbulence_Seamless}.G
    B=TextureSample(UVs=@n11) {texture=&T_Turbulence_Seamless}.G
    Alpha=@n3
   )) {constB=2}
  n13 = Multiply(A=@n12, B=@n6)
  n14 = Multiply(A=@n6) {constB=0.25}
  n15 = Multiply(A=Add(A=Multiply(A=@n2.R) {constB=0.2}, B=Add(A=@n13, B=@n14)), B=ScalarParameter() {defaultValue=0.4, parameterName=Opaciy})
  Multiply(A=@n1.A, B=@n15)
  n16 = Rotator(Coordinate=TextureCoordinate() {coordinateIndex=1}) {speed=0.5}
  n17 = Multiply(A=@n12) {constB=0.6}
  n18 = Multiply(A=TextureSample(UVs=Add(A=@n16, B=@n17)) {texture=/Game/Global/Textures/Default/Masks_and_Gradients/T_Grad_Rainbow}.RGB) {constB=0.2}
  n19 = Constant()
  n20 = MaterialFunctionCall!(Use Alternate UVs, Alternate UVs) {materialFunction=/Game/Athena/Effects/Materials/Geode/Functions/MF_Geode_Crack_Colors}
  n21 = Multiply(A=Add(A=QualitySwitch(Default=@n18, Low=@n19), B=@n20.Cloud Color), B=@n12)
  n22 = Power(Base=@n6) {constExponent=5}
  n23 = LinearInterpolate(
   A=TextureSample(UVs=@n10) {samplerType=SAMPLERTYPE_Normal, texture=/Game/Effects/Fort_Effects/Textures/Plasma/T_Turbulence_Seamless_N}.RGB
   B=TextureSample(UVs=@n11) {samplerType=SAMPLERTYPE_Normal, texture=&T_Turbulence_Seamless_N}.RGB
   Alpha=@n3
  )
  n24 = Multiply(A=@n4, B=Constant3Vector() {constant=<rgba 0 1 0 0>})
  n25 = Multiply(A=Clamp(None=DotProduct(A=@n23, B=@n24))) {constB=2}
  n26 = Constant()
  n27 = Multiply(A=Multiply(A=@n20.Crack Color, B=@n22), B=Add(A=QualitySwitch(Default=@n25, Low=@n26), B=@n22))
  Multiply(
   A=@n1.RGB
   B=LinearInterpolate(A=Add(A=@n21, B=@n27), B=VectorParameter() {parameterName=InnerColor, defaultValue=<rgba 0.2603 0.2552 1 0>}.RGB, Alpha=@n2.G) {constB=0}
  )
  QualitySwitch(
   Default=LinearInterpolate(B=ScalarParameter() {defaultValue=1.05, parameterName=Refraction}, Alpha=@n15) {constA=1, constB=1.2}
   Low=Constant() {r=1}
  )
  QualitySwitch(Default=@n23, Low=Constant())
 ]
}
